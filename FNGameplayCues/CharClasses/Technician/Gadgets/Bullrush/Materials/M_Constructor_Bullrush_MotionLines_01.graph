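# Translucent unlit particle sprite material for motion lines in the Constructor's Bullrush gadget effect, with responsive anti-aliasing enabled.
# Particle color RGB passes through an emissive adjustment function, and particle alpha scales a procedural, texture-free streak mask.
# The mask peaks at the center of U, rises along V and falls off near V=1.
# Both falloff multipliers use the same "Scale A" parameter name.
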Material /FNGameplayCues/CharClasses/Technician/Gadgets/Bullrush/Materials/M_Constructor_Bullrush_MotionLines_01
graph {
 properties {
  bEnableResponsiveAA=true
  bUsedWithParticleSprites=true
  blendMode=BLEND_Translucent
  pixelDepthOffsetMode=PDOM_Legacy
  shadingModel=MSM_Unlit
  usedShadingModels=Unlit
 }
 nodes [
  n1 = ParticleColor()
  MaterialFunctionCall(Emissive=@n1.RGB) {materialFunction=&MF_FX_EmissiveAdjustments}
  n2 = ComponentMask(None=TextureCoordinate()) {g=true}
  n3 = ComponentMask(None=TextureCoordinate()) {r=true}
  n4 = Multiply(A=Multiply(A=@n3, B=OneMinus(None=@n3)), B=ScalarParameter() {defaultValue=3, parameterName=Scale A})
  n5 = Multiply(A=OneMinus(None=@n2), B=ScalarParameter() {defaultValue=3, parameterName=Scale A})
  Multiply(A=@n1.A, B=Multiply(A=Multiply(A=@n2, B=@n4), B=Clamp(None=@n5)))
 ]
}
